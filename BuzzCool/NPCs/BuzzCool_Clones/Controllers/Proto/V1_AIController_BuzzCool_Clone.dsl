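; AI controller for BuzzCool clone NPCs that reacts to goal actor changes, writing hostility, token-provider and range flags to the blackboard.
; An item-owner goal registers the clone with that owner's spawn ability and marks it collected.
; For pawn goals, a looping timer re-centres the leash and pushes an equipped or marked-target-based state to the pawn, both on authority.
; Collected clones that lose their goal are force-killed after a timeout.
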(event EventBeginPlay
  (Development|PrintString "@@@ Event BeginPlay [AIController_BuzzCool_Clone]: CALLED!" false (Variables|Debug|GetbDebugController))
  (bind _returnvalue (AI|GoalSelection|GetAIGoalComponent self))
  (EventDispatchers|BindEventtoOnNewGoalActorDelegate _returnvalue (EventDispatchers|CreateEvent :!event Custom|OnCloneGoalActorChanged))
  (AI|Components|Blackboard|SetValueasBool (AI|GetBlackboard self) (Variables|Internal|GetGoalActorIsHostile_BBKey) false)
  (Variables|Internal|SetbCloneIsCollected false)
  (Variables|Internal|SetbGoalIsItemOwner false))

(event Custom|OnCloneGoalActorChanged
  (params (OldActor :type Actor%20Object%20Reference) (NewActor :type Actor%20Object%20Reference))
  (Development|PrintString (Utilities|String|Append "@@@@@@ OnCloneGoalActorChanged [AIController_BuzzCool_Clone]: CALLED; Goal Actor changed from: " (Utilities|GetDisplayName OldActor) ", to: " (Utilities|GetDisplayName NewActor) "!") false (Variables|Debug|GetbDebugController))
  (Utilities|IsValid NewActor
    (:Is%20Valid
      (CallFunction|StopPawnDeathPolling)
      (CallFunction|ProcessActorTarget NewActor)
      (if (Variables|Internal|GetbGoalIsAPawn)
        (Variables|Internal|SetCurrentGoalPawnActor NewActor)
        (CallFunction|StartPawnActorPolling)
        (else
          (CallFunction|StopPawnActorPolling)
          (CallFunction|UpdateCloneLeash NewActor))))
    (:Is%20Not%20Valid
      (Variables|Internal|SetCurrentGoalPawnActor 0)
      (CallFunction|StopPawnActorPolling)
      (if (Variables|Internal|GetbCloneIsCollected)
        (CallFunction|StartPawnDeathPolling)
        (Variables|Internal|SetbCloneIsCollected false)))))

(event Custom|StartPawnDeathPolling
  (Development|PrintString "@@@ StartPawnDeathPolling [AIController_BuzzCool_Clone]: CALLED!" false (Variables|Debug|GetbDebugController))
  (if (not (Utilities|Time|IsTimerActivebyHandle (Variables|Internal|GetTimerHandle_PawnDeathUpdate)))
    (bind _returnvalue (Utilities|Time|SetTimerbyEvent (delegate Custom|OnCollectedPawnTimeout) (ScalableFloat|GetValueatLevel (Variables|Internal|GetSF_CollectedPawnTimeout))))
    (Variables|Internal|SetTimerHandle_PawnDeathUpdate _returnvalue)))

(event Custom|StartPawnActorPolling
  (Development|PrintString "@@@ StartPawnActorPolling [AIController_BuzzCool_Clone]: CALLED!" false (Variables|Debug|GetbDebugController))
  (if (not (Utilities|Time|IsTimerActivebyHandle (Variables|Internal|GetTimerHandle_PawnActorUpdate)))
    (CallFunction|UpdateCloneLeash (Variables|Internal|GetCurrentGoalPawnActor))
    (CallFunction|UpdateCloneState)
    (bind _returnvalue (Utilities|Time|SetTimerbyEvent (delegate Custom|UpdateItemOwnerData) (ScalableFloat|GetValueatLevel (Variables|Internal|GetSF_ItemOwnerPollTime)) true))
    (Variables|Internal|SetTimerHandle_PawnActorUpdate _returnvalue)))

(event Custom|OnCollectedPawnTimeout
  (bind _self self)
  (Development|PrintString "@@@@@@ OnCollectedPawnTimeout [AIController_BuzzCool_Clone]: CALLED!" false (Variables|Debug|GetbDebugController))
  (bind _asfort_pawn (Utilities|Casting|CastToFortPawn (Pawn|GetControlledPawn _self)))
  (FortPawn|ForceKill _asfort_pawn "(TagName=\"Gameplay.Effect.InstantDeath.Despawn\")" _self _asfort_pawn))

(event Custom|UpdateItemOwnerData
  (Development|PrintString "@@@@@@ UpdateItemOwnerData [AIController_BuzzCool_Clone]: CALLED!" false (Variables|Debug|GetbDebugController))
  (CallFunction|UpdateCloneLeash (Variables|Internal|GetCurrentGoalPawnActor))
  (CallFunction|UpdateCloneState))

(event Custom|StopPawnActorPolling
  (Development|PrintString "@@@ StopPawnActorPolling [AIController_BuzzCool_Clone]: CALLED!" false (Variables|Debug|GetbDebugController))
  (Utilities|Time|ClearandInvalidateTimerbyHandle (Variables|Internal|GetTimerHandle_PawnActorUpdate)))

(event Custom|StopPawnDeathPolling
  (Development|PrintString "@@@ StopPawnDeathPolling [AIController_BuzzCool_Clone]: CALLED!" false (Variables|Debug|GetbDebugController))
  (if (Utilities|Time|IsTimerActivebyHandle (Variables|Internal|GetTimerHandle_PawnDeathUpdate))
    (Utilities|Time|ClearandInvalidateTimerbyHandle (Variables|Internal|GetTimerHandle_PawnDeathUpdate))))

(event EventOnGoalActorDistanceNotify
  (params (NotifyIndex :type Integer) (IsValid :type Boolean))
  (switch int NotifyIndex
    (:0
      (if (not (== IsValid (Variables|Internal|GetGoalActorDistance_IsInMeleeRange)))
        (bind _output_get (Variables|Internal|SetGoalActorDistance_IsInMeleeRange IsValid))
        (AI|Components|Blackboard|SetValueasBool (AI|GetBlackboard self) (Variables|Internal|GetGoalActorDistance_IsInMeleeRangeKey) _output_get)))
    (:1
      (if (not (== IsValid (Variables|Internal|GetGoalActorDistance_IsInCloseRange)))
        (bind _output_get_2 (Variables|Internal|SetGoalActorDistance_IsInCloseRange IsValid))
        (AI|Components|Blackboard|SetValueasBool (AI|GetBlackboard self) (Variables|Internal|GetGoalActorDistance_IsInCloseRangeKey) _output_get_2)))))

(fn ProcessActorTarget
  (params (TargetActor :type Actor%20Object%20Reference))
  (bind _self self)
  (bind _returnvalue (AI|GetBlackboard _self))
  (bind _currentgoalpawn (LocalVariables|GetCurrentGoalPawn))
  (bind _returnvalue_2 (Ability|GetAbilitySystemComponent _currentgoalpawn))
  (bind _targetactor (Variables|GetTargetActor))
  (Development|PrintString "@@@@@@@@@ ProcessActorTarget [AIController_BuzzCool_Clone]: CALLED!" false (Variables|Debug|GetbDebugController))
  (bind _returnvalue_3 (GameplayTags|HasMatchingGameplayTag(Message) _targetactor "(TagName=\"Ability.BuzzCool.Item.MarkedTarget\")"))
  (LocalVariables|SetbIsMarkedTarget _returnvalue_3)
  (Utilities|IsA(soft) _targetactor "/Script/FortniteGame.FortPlayerPawn"
    (:True
      (Variables|Internal|SetbGoalIsAPawn true)
      (bind _asfort_player_pawn (Utilities|Casting|CastToFortPlayerPawn _targetactor)
        (:then
          (bind _output_get (LocalVariables|SetCurrentGoalPawn _asfort_player_pawn))
          (bind _returnvalue_4 (GameplayTags|HasMatchingGameplayTag(Message) _output_get "(TagName=\"Ability.BuzzCool.Item.Owner\")"))
          (if _returnvalue_4
            (Variables|Internal|SetbCloneIsCollected true)
            (Variables|Internal|SetbGoalIsItemOwner true)
            (bind _outabilityhandles (GameplayAbilities|FindAllAbilitieswithTags _returnvalue_2 "(GameplayTags=((TagName=\"Ability.BuzzCool.ItemOwner.Spawn\")))" true))
            (if (Utilities|Array|IsNotEmpty _outabilityhandles)
              (bind _asga_buzz_cool_clone_spawn_item_owner (Utilities|Casting|CastToGA_BuzzCool_CloneSpawn_ItemOwner (.ReturnValue (Ability|GameplayAbility|GetGameplayAbilityfromSpecHandle _returnvalue_2 (Utilities|Array|Get(acopy) _outabilityhandles)))))
              (Class|GABuzzCoolCloneSpawnBase|AddUniqueCloneActor _asga_buzz_cool_clone_spawn_item_owner (Pawn|GetControlledPawn self)))
            (else
              (Variables|Internal|SetbGoalIsItemOwner false))))
        (:CastFailed
          _)))
    (:False
      (Variables|Internal|SetbGoalIsAPawn false)
      (LocalVariables|SetCurrentGoalPawn 0)))
  (bind _output_get_2 (Variables|Internal|SetbHasHostileGoal (or (and (Variables|Internal|GetbGoalIsAPawn) (Utilities|Enum|Equal(Enum) (Team|GetTeamAffiliation (Pawn|GetControlledPawn self) (LocalVariables|GetCurrentGoalPawn)) "Hostile")) (LocalVariables|GetbIsMarkedTarget))))
  (AI|Components|Blackboard|SetValueasBool _returnvalue (Variables|Internal|GetGoalActorIsHostile_BBKey) _output_get_2)
  (bind _output_get_3 (Variables|Internal|SetbHasTokenProvider (Utilities|IsValid (Actor|GetComponentbyClass (Variables|GetTargetActor) "/AITokenProvider/Assets/BP_TokenProvider_Default.BP_TokenProvider_Default_C") :!class /Script/Engine.KismetSystemLibrary)))
  (AI|Components|Blackboard|SetValueasBool _returnvalue (Variables|Internal|GetGoalActorHasTokenProvider_BBKey) _output_get_3))

(fn UpdateCloneLeash
  (params (LeashToActor :type Actor%20Object%20Reference))
  (bind _fortathenaleash (Variables|Components|GetFortAthenaLeash))
  (bind _fortathenaleash_2 (Variables|Components|GetFortAthenaLeash))
  (Networking|SwitchHasAuthority
    (:Authority
      (Utilities|IsValid LeashToActor
        (:Is%20Valid
          (Class|FortAthenaLeashComponent|SetLeashCenter _fortathenaleash_2 (Transformation|GetActorLocation LeashToActor)))
        (:Is%20Not%20Valid
          (Class|FortAthenaLeashComponent|SetLeashCenter _fortathenaleash_2 (Pawn|GetNavAgentLocation (Pawn|GetControlledPawn self)))))
      (Class|FortAthenaLeashComponent|SetLeashRadius _fortathenaleash_2 (ScalableFloat|GetValueatLevel (Variables|Internal|GetSF_ActorLeash_Radius)))
      (if (Variables|Debug|GetbDebugController)
        (Utilities|IsValid _fortathenaleash
          (:Is%20Valid
            (Rendering|Debug|DrawDebugSphere (Class|FortAthenaLeashComponent|GetLeashCenter _fortathenaleash) (Class|FortAthenaLeashComponent|GetLeashRadius _fortathenaleash) 50 "(R=0.978654,G=0.019030,B=1.000000,A=1.000000)" 5.0 5.0)))))))

(fn UpdateCloneState
  (bind _currentgoalpawnactor (Variables|Internal|GetCurrentGoalPawnActor))
  (Networking|SwitchHasAuthority
    (:Authority
      (Utilities|IsValid (Variables|Internal|GetCurrentGoalPawnActor)
        (:Is%20Valid
          (if (Variables|Internal|GetbGoalIsItemOwner)
            (bind _returnvalue (GameplayTags|HasMatchingGameplayTag(Message) _currentgoalpawnactor "(TagName=\"Ability.BuzzCool.Item.Equipped\")"))
            (if _returnvalue
              (bind _output_get (Variables|Internal|SetEnum_PawnGoalState "NewEnumerator1"))
              (LocalVariables|Setloc_PawnGoalState _output_get)
              (else
                (bind _output_get_2 (Variables|Internal|SetEnum_PawnGoalState "NewEnumerator0"))
                (LocalVariables|Setloc_PawnGoalState _output_get_2)))
            (else
              (bind _returnvalue_2 (GameplayTags|HasMatchingGameplayTag(Message) _currentgoalpawnactor "(TagName=\"Ability.BuzzCool.Item.MarkedTarget\")"))
              (if _returnvalue_2
                (bind _output_get_3 (Variables|Internal|SetEnum_PawnGoalState "NewEnumerator2"))
                (LocalVariables|Setloc_PawnGoalState _output_get_3)
                (else
                  (bind _output_get_4 (Variables|Internal|SetEnum_PawnGoalState "NewEnumerator3"))
                  (LocalVariables|Setloc_PawnGoalState _output_get_4)))))
          (Development|PrintString (Utilities|String|Append "@@@@@@@@@ UpdateCloneState [AIController_BuzzCool_Clone]: CALLED - Pawn Goal State = " (Utilities|String|EnumtoString (LocalVariables|Getloc_PawnGoalState)) "!") false (Variables|Debug|GetbDebugController))
          (bind _aspawn_wild_estate_buzz_cool (Utilities|Casting|CastToPawn_WildEstate_BuzzCool (Pawn|GetControlledPawn self)))
          (Variables|/BuzzCool/NPCs/BuzzCool_Clones/Pawn/Pawn_WildEstate_BuzzCool.Pawn_WildEstate_BuzzCool_C|SetEnum%20Pawn%20Goal%20State (LocalVariables|Getloc_PawnGoalState) _aspawn_wild_estate_buzz_cool))))))

(fn ConstructionScript)
